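# Physics asset for the base Husk enemy mesh, defining simple collision bodies and jointed constraints for its skeleton.
# Capsules and spheres cover the torso, head and limbs, with boxes for hands and feet, chained outward from the pelvis.
# Elbows and knees are restricted to hinge-like swing, while the ankles twist freely.
# A two-body hood chain hangs from the upper spine with one swing axis locked.
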PhysicsAsset /Game/Characters/Enemies/Husk/Mesh/Husk/Base_Husk_Physics
bodies {
 Spine_01 [{shapeType=Capsule, length=8.1735, radius=9.2167, center=<5.7127 0.5639 -0.1362>, rotation=<rot 0 94.4965 -90.0001>}]
 Spine_02 [{shapeType=Capsule, length=8.1174, radius=12.7498, center=<5.9361 -0.5613 -0.1362>, rotation=<rot 0 89.4657 -90.0001>}]
 Spine_03 [{shapeType=Sphere, radius=13.3583, center=<4.2994 -1.087 -0.1362>}]
 calf_l [{shapeType=Capsule, length=30, radius=8.5, center=<-19.2113 0 -0.0001>, rotation=<rot -90 -38.9387 38.9388>}]
 calf_r [{shapeType=Capsule, length=30, radius=8.5, center=<18.693 0.0001 0>, rotation=<rot -90 -38.9387 38.9388>}]
 foot_l [{shapeType=Box, extentX=9.5979, extentY=10.8534, extentZ=27.2828, center=<-5.3152 7.6521 0.0768>, rotation=<rot 0.6692 -179.4029 -88.9983>}]
 foot_r [{shapeType=Box, extentX=9.5979, extentY=10.8534, extentZ=27.2828, center=<5.3766 -7.8657 -0.3194>, rotation=<rot 0.6692 -179.4029 -88.9983>}]
 hand_l [{shapeType=Box, extentX=15.4793, extentY=7.5682, extentZ=11.0177, center=<9.4811 -2.0815 0.5808>}]
 hand_r [{shapeType=Box, extentX=15.4793, extentY=7.5682, extentZ=9.5177, center=<-10.0681 2.0816 -1.0805>}]
 head [{shapeType=Capsule, length=10.0736, radius=11.1945, center=<1.6183 5.4619 -0.1362>, rotation=<rot 0.0015 70.7256 -89.9996>}]
 hood_01 [{shapeType=Capsule, length=11.6057, radius=12.967, center=<0 -2.433 2.6793>, rotation=<rot 0 0 -30>}]
 hood_02 [{shapeType=Sphere, radius=16.0765, center=<0 -11.6448 0.9966>}]
 lowerarm_l [{shapeType=Capsule, length=20, radius=5.5, center=<10.0133 1.6278 0>, rotation=<rot -90 -167.068 167.068>}]
 lowerarm_r [{shapeType=Capsule, length=20, radius=5.5, center=<-11.5596 -2.5208 0.0001>, rotation=<rot -90 -167.068 167.068>}]
 pelvis [{shapeType=Sphere, radius=13.0077, center=<6.8841 0.9717 0>}]
 thigh_l [{shapeType=Capsule, length=28.5, radius=6.5, center=<-17.8209 0 1.1552>, rotation=<rot -90 6.1741 -6.174>}]
 thigh_r [{shapeType=Capsule, length=28.5, radius=6.5, center=<15.9401 0.0001 -1.0641>, rotation=<rot -90 6.1741 -6.174>}]
 upperarm_l [{shapeType=Capsule, length=16, radius=7, center=<7.7352 1.3911 0>, rotation=<rot -90 -75.392 75.3921>}]
 upperarm_r [{shapeType=Capsule, length=16, radius=7, center=<-7.9328 -2.7241 0>, rotation=<rot -90 -75.392 75.3921>}]
}
constraints [
 {bone1Name=Spine_01, bone2Name=pelvis, swing1LimitDegrees=45, swing2LimitDegrees=45, twistLimitDegrees=45}
 {bone1Name=Spine_02, bone2Name=Spine_01, swing1LimitDegrees=45, swing2LimitDegrees=45, twistLimitDegrees=45}
 {bone1Name=Spine_03, bone2Name=Spine_02, swing1LimitDegrees=45, swing2LimitDegrees=45, twistLimitDegrees=45}
 {bone1Name=upperarm_l, bone2Name=Spine_03, swing1LimitDegrees=45, swing2LimitDegrees=45, twistLimitDegrees=45}
 {
  bone1Name=lowerarm_l
  bone2Name=upperarm_l
  swing1LimitDegrees=60
  swing2LimitDegrees=45
  swing2Motion=Locked
  twistLimitDegrees=45
  twistMotion=Locked
 }
 {bone1Name=hand_l, bone2Name=lowerarm_l, swing1LimitDegrees=45, swing2LimitDegrees=45, twistLimitDegrees=45}
 {bone1Name=upperarm_r, bone2Name=Spine_03, swing1LimitDegrees=45, swing2LimitDegrees=45, twistLimitDegrees=45}
 {
  bone1Name=lowerarm_r
  bone2Name=upperarm_r
  swing1LimitDegrees=60
  swing2LimitDegrees=45
  swing2Motion=Locked
  twistLimitDegrees=45
  twistMotion=Locked
 }
 {bone1Name=hand_r, bone2Name=lowerarm_r, swing1LimitDegrees=45, swing2LimitDegrees=45, twistLimitDegrees=45}
 {bone1Name=head, bone2Name=Spine_03, swing1LimitDegrees=45, swing2LimitDegrees=45, twistLimitDegrees=45}
 {bone1Name=thigh_l, bone2Name=pelvis, swing1LimitDegrees=45, swing2LimitDegrees=45, twistLimitDegrees=45}
 {bone1Name=calf_l, bone2Name=thigh_l, swing1LimitDegrees=60, swing2LimitDegrees=0, twistLimitDegrees=0}
 {bone1Name=foot_l, bone2Name=calf_l, swing1LimitDegrees=45, swing2LimitDegrees=45, twistLimitDegrees=45, twistMotion=Free}
 {bone1Name=thigh_r, bone2Name=pelvis, swing1LimitDegrees=45, swing2LimitDegrees=45, twistLimitDegrees=45}
 {bone1Name=calf_r, bone2Name=thigh_r, swing1LimitDegrees=60, swing2LimitDegrees=0, twistLimitDegrees=0}
 {bone1Name=foot_r, bone2Name=calf_r, swing1LimitDegrees=45, swing2LimitDegrees=45, twistLimitDegrees=45, twistMotion=Free}
 {bone1Name=hood_01, bone2Name=Spine_03, swing1LimitDegrees=57.7434, swing2LimitDegrees=45, swing2Motion=Locked, twistLimitDegrees=62.5221}
 {bone1Name=hood_02, bone2Name=hood_01, swing1LimitDegrees=62.5221, swing2LimitDegrees=45, swing2Motion=Locked, twistLimitDegrees=54.5575}
]
