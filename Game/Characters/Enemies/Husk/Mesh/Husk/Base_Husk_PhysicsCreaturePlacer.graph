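# Physics asset for the base Husk enemy mesh, a simplified variant named for a physics creature placer.
# It has a single large capsule body on the pelvis bone, pitched 90 degrees and slightly offset from the bone origin.
# No other bones have bodies and there are no constraints, so the asset represents the creature as one rigid shape rather than a jointed ragdoll.
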PhysicsAsset /Game/Characters/Enemies/Husk/Mesh/Husk/Base_Husk_PhysicsCreaturePlacer
bodies {pelvis [{shapeType=Capsule, length=84.25, radius=38.75, center=<-3.8037 -0.0004 0>, rotation=<rot -90 0.0396 -0.0396>}]}
constraints []
