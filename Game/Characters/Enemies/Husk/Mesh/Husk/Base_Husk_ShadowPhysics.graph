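# Physics asset for the Husk enemy's skeleton, named as its shadow physics variant, approximating the humanoid body with one capsule per bone.
# Capsules cover pelvis, spine, head, hood, arms, hands, legs and feet, a capsule-only layout that suits capsule shadows.
# Constraints chain the bodies through the spine and pelvis, with spine, hip and wrist joints fully free.
# Elbows lock twist and one swing axis, the hood locks one swing, and ankles free twist.
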PhysicsAsset /Game/Characters/Enemies/Husk/Mesh/Husk/Base_Husk_ShadowPhysics
bodies {
 Spine_01 [{shapeType=Capsule, length=8.1735, radius=9.2167, center=<5.7127 0.5639 -1.1516>, rotation=<rot 0 94.4965 -90.0001>}]
 calf_l [{shapeType=Capsule, length=39.0586, radius=4.9832, center=<-19.2113 0 -0.0001>, rotation=<rot -90 -38.9387 38.9388>}]
 calf_r [{shapeType=Capsule, length=41.7678, radius=4.8616, center=<18.693 0.0001 0>, rotation=<rot -90 -38.9387 38.9388>}]
 foot_l [{shapeType=Capsule, length=23.3673, radius=6.6327, center=<-6.825 8.3657 -0.4342>, rotation=<rot 2.1696 8.0542 78.653>}]
 foot_r [{shapeType=Capsule, length=16.7314, radius=7.9546, center=<5.9912 -9.8924 -1.0247>, rotation=<rot 0.9607 9.3163 84.164>}]
 hand_l [{shapeType=Capsule, length=11.8762, radius=5.106, center=<8.933 -2.5193 0.1414>, rotation=<rot 0 -104.0036 -88.4171>}]
 hand_r [{shapeType=Capsule, length=12.6814, radius=5.3593, center=<-8.4382 2.6943 -0.6856>, rotation=<rot 0 -102.4978 -79.92>}]
 head [{shapeType=Capsule, length=12.1382, radius=9.1299, center=<1.6183 5.4619 -0.1362>, rotation=<rot 0.0015 70.7256 -89.9996>}]
 hood_01 [{shapeType=Capsule, length=11.6057, radius=12.967, center=<0 -7.0488 -3.4281>, rotation=<rot 0 0 -100.132>}]
 lowerarm_l [{shapeType=Capsule, length=21.5099, radius=3.9901, center=<9.8856 0.3601 0.0001>, rotation=<rot -90 -155.5175 149.7639>}]
 lowerarm_r [{shapeType=Capsule, length=22.2011, radius=3.2989, center=<-11.5596 -2.5208 0.0001>, rotation=<rot -90 -167.068 167.068>}]
 pelvis [{shapeType=Capsule, length=1.094, radius=13.3808, center=<4.2857 -0.1397 0.1287>, rotation=<rot 1.0119 -90 -1.8492>}]
 spine_03 [{shapeType=Capsule, length=17.0586, radius=12.2849, center=<0.0775 -3.0858 -0.135>, rotation=<rot -0.0052 -104.4079 -89.9885>}]
 thigh_l [{shapeType=Capsule, length=39.575, radius=5.6004, center=<-17.8209 0 1.1552>, rotation=<rot -90 6.1741 -6.174>}]
 thigh_r [{shapeType=Capsule, length=37.16, radius=4.6904, center=<15.9146 0.0001 -2.0195>, rotation=<rot -88.4706 179.9998 -179.9998>}]
 upperarm_l [{shapeType=Capsule, length=25.9454, radius=4.075, center=<12.6897 1.391 0.0001>, rotation=<rot -90 -75.392 75.3921>}]
 upperarm_r [{shapeType=Capsule, length=27.431, radius=4.7433, center=<-12.0645 -2.7242 -0.0001>, rotation=<rot -90 -75.392 75.3921>}]
}
constraints [
 {bone1Name=upperarm_l, bone2Name=spine_03, swing1LimitDegrees=45, swing2LimitDegrees=45, twistLimitDegrees=45}
 {
  bone1Name=lowerarm_l
  bone2Name=upperarm_l
  swing1LimitDegrees=60
  swing2LimitDegrees=45
  swing2Motion=Locked
  twistLimitDegrees=45
  twistMotion=Locked
 }
 {bone1Name=upperarm_r, bone2Name=spine_03, swing1LimitDegrees=45, swing2LimitDegrees=45, twistLimitDegrees=45}
 {
  bone1Name=lowerarm_r
  bone2Name=upperarm_r
  swing1LimitDegrees=60
  swing2LimitDegrees=45
  swing2Motion=Locked
  twistLimitDegrees=45
  twistMotion=Locked
 }
 {bone1Name=head, bone2Name=spine_03, swing1LimitDegrees=45, swing2LimitDegrees=45, twistLimitDegrees=45}
 {bone1Name=calf_l, bone2Name=thigh_l, swing1LimitDegrees=60, swing2LimitDegrees=0, twistLimitDegrees=0}
 {bone1Name=foot_l, bone2Name=calf_l, swing1LimitDegrees=45, swing2LimitDegrees=45, twistLimitDegrees=45, twistMotion=Free}
 {bone1Name=calf_r, bone2Name=thigh_r, swing1LimitDegrees=60, swing2LimitDegrees=0, twistLimitDegrees=0}
 {bone1Name=foot_r, bone2Name=calf_r, swing1LimitDegrees=45, swing2LimitDegrees=45, twistLimitDegrees=45, twistMotion=Free}
 {bone1Name=hood_01, bone2Name=spine_03, swing1LimitDegrees=57.7434, swing2LimitDegrees=45, swing2Motion=Locked, twistLimitDegrees=62.5221}
 {
  bone1Name=spine_03
  bone2Name=Spine_01
  swing1LimitDegrees=45
  swing1Motion=Free
  swing2LimitDegrees=45
  swing2Motion=Free
  twistLimitDegrees=45
  twistMotion=Free
 }
 {
  bone1Name=Spine_01
  bone2Name=pelvis
  swing1LimitDegrees=45
  swing1Motion=Free
  swing2LimitDegrees=45
  swing2Motion=Free
  twistLimitDegrees=45
  twistMotion=Free
 }
 {
  bone1Name=thigh_l
  bone2Name=pelvis
  swing1LimitDegrees=45
  swing1Motion=Free
  swing2LimitDegrees=45
  swing2Motion=Free
  twistLimitDegrees=45
  twistMotion=Free
 }
 {
  bone1Name=thigh_r
  bone2Name=pelvis
  swing1LimitDegrees=45
  swing1Motion=Free
  swing2LimitDegrees=45
  swing2Motion=Free
  twistLimitDegrees=45
  twistMotion=Free
 }
 {
  bone1Name=hand_l
  bone2Name=lowerarm_l
  swing1LimitDegrees=45
  swing1Motion=Free
  swing2LimitDegrees=45
  swing2Motion=Free
  twistLimitDegrees=45
  twistMotion=Free
 }
 {
  bone1Name=hand_r
  bone2Name=lowerarm_r
  swing1LimitDegrees=45
  swing1Motion=Free
  swing2LimitDegrees=45
  swing2Motion=Free
  twistLimitDegrees=45
  twistMotion=Free
 }
]
